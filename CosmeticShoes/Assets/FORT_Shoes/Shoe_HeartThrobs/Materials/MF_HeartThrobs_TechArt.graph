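# Material function for the HeartThrobs cosmetic shoe that makes heart shapes glow and pulse to audio, with an idle beating cycle for gameplay.
# A distance-field texture drives rings, fades and erosion-based expansion.
# Shape-index texture and pre-skinned local position separate hearts: four side hearts follow spectrum bands, back and bottom ones follow amplitude.
# The masked result overrides BaseColor, Roughness and Emissive on the incoming material attributes.
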MaterialFunction /CosmeticShoes/Assets/FORT_Shoes/Shoe_HeartThrobs/Materials/MF_HeartThrobs_TechArt
graph {nodes [
  n1 = GetMaterialAttributes(None=NamedRerouteUsage()) {attributeGetTypes [#1, #2, #3]}
  FunctionOutput(None=SetMaterialAttributes!(MaterialAttributes=@n1.MaterialAttributes, BaseColor=NamedRerouteUsage(), Roughness=NamedRerouteUsage(), EmissiveColor=NamedRerouteUsage()) {attributeSetTypes [#1, #2, #3]})
  n2 = FunctionInput(Preview=TextureSample() {samplerType=SAMPLERTYPE_Masks, texture=/CosmeticShoes/Assets/FORT_Shoes/Shoe_HeartThrobs/Textures/T_HeartThrobs_Shoe_FX}.RGB) {inputName=In_TechArtMask, sortPriority=1}
  NamedRerouteDeclaration(None=ComponentMask(None=NamedRerouteDeclaration(None=@n2) {name=In_TechArtMask, nodeColor=<rgba 1 0.6797 0 1>}) {g=true}) {name=Hearts Emissive Mask, nodeColor=<rgba 0 1 0.2969 1>}
  NamedRerouteDeclaration(None=TextureSample() {samplerType=SAMPLERTYPE_LinearGrayscale, texture=/CosmeticShoes/Assets/FORT_Shoes/Shoe_HeartThrobs/Textures/T_HeartThrobs_FX_Hearts}.R) {name=Hearts Distance Field, nodeColor=<rgba 0 1 0.6953 1>}
  n3 = NamedRerouteUsage()
  n4 = OneMinus(None=@n3)
  n5 = ScalarParameter() {defaultValue=6, group=Audio Reactivity - Patterms and Masks, parameterName=Emissive Vignette Exponent}
  NamedRerouteDeclaration(None=Power(Base=@n3, Exp=Power(Base=@n4, Exp=@n5))) {name=Emissive Vignette, desc=Nice soft falloff for the emissive, nodeColor=<rgba 1 0 0.375 1>}
  n6 = Saturate(None=Divide(A=@n3, B=ScalarParameter() {defaultValue=0.03, group=Audio Reactivity - Patterms and Masks, parameterName=Base Color Fade Size}) {constB=0.1})
  NamedRerouteDeclaration(None=Power(
    Base=OneMinus(None=@n6)
    Exp=ScalarParameter() {defaultValue=9, group=Audio Reactivity - Patterms and Masks, parameterName=Base Color Fade Exponent}
   )) {name=Base Color Fade Mask, desc=Use me to modulate any PBR values you might need for the heart shapes, nodeColor=<rgba 0 0.6953 1 1>}
  n7 = Power(
   Base=NamedRerouteUsage()
   Exp=ScalarParameter() {defaultValue=1, group=Audio Reactivity - Patterms and Masks, parameterName=Heart Rings Spacing Exponent}
  )
  n8 = Frac(None=Multiply(A=OneMinus(None=@n7), B=ScalarParameter() {defaultValue=3, group=Audio Reactivity - Patterms and Masks, parameterName=Heart Rings Amount}) {constB=3})
  n9 = ScalarParameter() {defaultValue=2.5, group=Audio Reactivity - Patterms and Masks, parameterName=Heart Rings Gradient Exponent}
  n10 = Power(
   Base=MaterialFunctionCall!(
    Gradient=OneMinus(None=@n8)
    Center Point=ScalarParameter() {defaultValue=0.2, group=Audio Reactivity - Patterms and Masks, parameterName=Heart Rings Gradient Center Point}
   ) {materialFunction=/Game/Effects/Fort_Effects/MaterialFunctions/MF_GradientReflected}.Result
   Exp=@n9
  )
  NamedRerouteDeclaration(None=ComponentMask(None=MaterialFunctionCall(In=@n10, Power=@n9) {materialFunction=&SCurve}.Result) {r=true}) {name=Heart Rings Mask, nodeColor=<rgba 1 0.3281 0 1>}
  n11 = VectorParameter() {group=Audio Reactivity - Color controls, parameterName=Emissive Color Heart Fill, defaultValue=<rgba 0.8 0 0.4313 0>}
  n12 = VectorParameter() {group=Audio Reactivity - Color controls, parameterName=Emissive Color Heart Rings, defaultValue=<rgba 3 0.4609 1.5649 0>}
  NamedRerouteDeclaration(None=LinearInterpolate(A=@n11.RGB, B=@n12.RGB, Alpha=NamedRerouteUsage())) {name=Hearts Emissive Color, nodeColor=<rgba 0.4062 0 1 1>}
  n13 = ScalarParameter() {defaultValue=0.22, group=Audio Reactivity - Animation Controls, parameterName=Hearts Expanding Static Core Size}
  n14 = Subtract(A=OneMinus(None=NamedRerouteUsage()), B=@n13)
  n15 = ScalarParameter() {defaultValue=0.1, group=Audio Reactivity - Animation Controls, parameterName=Hearts Expanding Static Core Softness}
  n16 = OneMinus(None=Saturate(None=Divide(A=@n14, B=@n15)))
  n17 = ScalarParameter() {defaultValue=0.3, group=Audio Reactivity - Animation Controls, parameterName=Hearts Expanding Static Core Brightness}
  n18 = OneMinus(None=Divide(A=@n14, B=OneMinus(None=@n13)))
  n19 = OneMinus(None=NamedRerouteUsage())
  n20 = ScalarParameter() {defaultValue=2, group=Audio Reactivity - Animation Controls, parameterName=Hearts Expanding Ring Width}
  NamedRerouteDeclaration(None=Max(
    A=Multiply(A=@n16, B=@n17) {constB=0.5}
    B=MaterialFunctionCall!(Gradient=@n18, Erosion Amount=@n19, Softness=@n20) {materialFunction=/BRCosmetics/Athena/Effects/Particles/Skydiving/NS_Contrail_DarkAndLight/Functions/MF_Erosion}.Increased Range
   )) {name=Hearts Expanding Mask, nodeColor=<rgba 0.2422 0 1 1>}
  n21 = NamedRerouteUsage()
  n22 = Power(Base=NamedRerouteUsage(), Exp=NamedRerouteUsage())
  NamedRerouteDeclaration(None=LinearInterpolate(A=NamedRerouteUsage(), B=DotProduct(A=@n21, B=@n22), Alpha=NamedRerouteUsage()) {desc="4 hearts on the sides react to spectrum,\r\nhearts on the back and bottom react to amplitude"}) {name=Reactivity Timeline, nodeColor=<rgba 0.1953 1 0 1>}
  n23 = NamedRerouteUsage()
  n24 = ConstantBiasScale(None=Sine(None=NamedRerouteUsage()))
  n25 = ScalarParameter() {defaultValue=2, group=Audio Reactivity - Gameplay Idle Cycle, parameterName=Idle Heart Beating Timeline Exponent}
  n26 = Multiply(
   A=OneMinus(None=Power(Base=@n24, Exp=@n25))
   B=ScalarParameter() {defaultValue=0.75, group=Audio Reactivity - Gameplay Idle Cycle, parameterName=Idle Heart Beating Max Erosion}
  )
  n27 = ScalarParameter() {defaultValue=0.25, group=Audio Reactivity - Gameplay Idle Cycle, parameterName=Idle Heart Beating Erosion Softness}
  NamedRerouteDeclaration(None=Multiply(
    A=MaterialFunctionCall(Gradient=@n23, Erosion Amount=@n26, Softness=@n27) {materialFunction=&MF_Erosion}.Increased Range
    B=ScalarParameter() {defaultValue=0.666, group=Audio Reactivity - Gameplay Idle Cycle, parameterName=Idle Heart Beating Brightness Multiplier}
   )) {name=Idle Hearts Beating Mask, nodeColor=<rgba 0 0.9531 1 1>}
  n28 = TextureSample() {
   samplerType=SAMPLERTYPE_LinearGrayscale
   texture=/CosmeticShoes/Assets/FORT_Shoes/Shoe_HeartThrobs/Textures/T_HeartThrobs_FX_Hearts_ShapeData
  }
  NamedRerouteDeclaration(None=@n28.R) {name=Shape Index, nodeColor=<rgba 1 0.8906 0 1>}
  n29 = VectorParameter() {group=Audio Reactivity - Color controls, parameterName=Emissive Color Baseline, defaultValue=<rgba 0.13 0 0.0701 0>}
  NamedRerouteDeclaration(None=Multiply(A=@n29.RGB, B=NamedRerouteUsage())) {name=Hearts Emissive Baseline Color, nodeColor=<rgba 0.8984 0 1 1>}
  n30 = NamedRerouteUsage()
  n31 = If(A=NamedRerouteUsage(), "A > B"=NamedRerouteUsage(), "A < B"=NamedRerouteUsage()) {constB=0.01}
  NamedRerouteDeclaration(None=LinearInterpolate(A=NamedRerouteUsage(), B=NamedRerouteUsage(), Alpha=Multiply(A=@n30, B=@n31))) {name=EmissiveColorFinal, nodeColor=<rgba 1 0 0.3984 1>}
  NamedRerouteDeclaration(None=ScalarParameter() {defaultValue=0.3, group=Audio Reactivity - Color controls, parameterName=Roughness on Hearts}) {name=RoughnessFinal, nodeColor=<rgba 0 0.7188 1 1>}
  n32 = ScalarParameter() {defaultValue=0.05, group=Audio Reactivity - Color controls, parameterName=Base Color Multiplier on Hearts}
  n33 = NamedRerouteUsage()
  NamedRerouteDeclaration(None=Multiply(A=NamedRerouteUsage(), B=LinearInterpolate(B=@n32, Alpha=@n33) {constA=1})) {name=BaseColorFinal, nodeColor=<rgba 1 0.9609 0 1>}
  NamedRerouteDeclaration(None=@n1.BaseColor) {name=In_BaseColor, nodeColor=<rgba 0.6406 0 1 1>}
  NamedRerouteDeclaration(None=@n1.Roughness) {name=In_Roughness, nodeColor=<rgba 0 1 0.3203 1>}
  NamedRerouteDeclaration(None=@n1.EmissiveColor) {name=In_Emissive, nodeColor=<rgba 1 0 0 1>}
  NamedRerouteDeclaration(None=FunctionInput() {inputName=In_MaterialAttributes, inputType=FunctionInput_MaterialAttributes}) {name=In_MaterialAttributes, nodeColor=<rgba 0 1 1 1>}
  n34 = NamedRerouteUsage()
  NamedRerouteDeclaration(None=LinearInterpolate(A=NamedRerouteUsage(), B=NamedRerouteUsage(), Alpha=@n34)) {name=Out_BaseColor, nodeColor=<rgba 1 0.0703 0 1>}
  NamedRerouteDeclaration(None=LinearInterpolate(A=NamedRerouteUsage(), B=NamedRerouteUsage(), Alpha=@n34)) {name=Out_Roughness, nodeColor=<rgba 1 0.0703 0 1>}
  NamedRerouteDeclaration(None=LinearInterpolate(A=NamedRerouteUsage(), B=NamedRerouteUsage(), Alpha=@n34)) {name=Out_Emissive, nodeColor=<rgba 1 0 0.8906 1>}
  n35 = MaterialFunctionCall!(InAmplitude, InSpectrum, CustomFakeSpectrum, CustomFakeAmplitude) {materialFunction=/BRCosmetics/Global/MaterialFunctions/AudioReactivity/MF_AudioReactivity_FrontendLoop}
  NamedRerouteDeclaration(None=@n35.Spectrum) {name=In_Spectrum, nodeColor=<rgba 0.7344 0 1 1>}
  NamedRerouteDeclaration(None=@n35.Amplitude) {name=Amplitude, nodeColor=<rgba 0.2188 1 0 1>}
  n36 = VertexInterpolator(VS=LocalPosition() {localOrigin=InstancePreSkinning}.XYZ)
  n37 = Ceil(None=Add(
    A=ComponentMask(None=@n36.PS) {g=true}
    B=ScalarParameter() {defaultValue=9, group=Audio Reactivity - Patterms and Masks, parameterName=Back Hearts Mask Position}
   ))
  n38 = ComponentMask(None=@n36.PS) {b=true}
  n39 = ScalarParameter() {defaultValue=-0.5, group=Audio Reactivity - Patterms and Masks, parameterName=Bottom Hearts Mask Position}
  n40 = Saturate(None=Ceil(None=Add(A=@n38, B=@n39)))
  NamedRerouteDeclaration(None=Multiply(A=Saturate(None=@n37), B=@n40)) {name=Side Hearts Mask, nodeColor=<rgba 1 0.2812 0 1>}
  NamedRerouteDeclaration(None=OneMinus(None=@n40)) {name=Bottom Hearts Mask, nodeColor=<rgba 0 1 0.7422 1>}
  n41 = VectorParameter() {group=Audio Reactivity - Spectrum and Amplitude controls, parameterName=Spectrum Side Heart 4, defaultValue=<rgba 1 0 0 0>}
  n42 = VectorParameter() {group=Audio Reactivity - Spectrum and Amplitude controls, parameterName=Spectrum Side Heart 3, defaultValue=<rgba 0 1 0 0>}
  n43 = VectorParameter() {group=Audio Reactivity - Spectrum and Amplitude controls, parameterName=Spectrum Side Heart 2, defaultValue=<rgba 0 0 1 0>}
  n44 = VectorParameter() {group=Audio Reactivity - Spectrum and Amplitude controls, parameterName=Spectrum Side Heart 1, defaultValue=<rgba 0 0 0 1>}
  NamedRerouteDeclaration(None=MaterialFunctionCall!(1=@n41.RGBA, 2=@n42.RGBA, 3=@n43.RGBA, 4=@n44.RGBA, A=NamedRerouteUsage()) {materialFunction=/Engine/Functions/Engine_MaterialFunctions02/Lerp_Multiple_Float4}.Lerp 4 Inputs) {name=Spectrum Mask, nodeColor=<rgba 0 1 1 1>}
  n45 = VectorParameter() {group=Audio Reactivity - Spectrum and Amplitude controls, parameterName=Spectrum Exponent Side Hearts, defaultValue=<rgba 2.25 2 1.75 1.5>}
  NamedRerouteDeclaration(None=MaterialFunctionCall!(1=@n45.R, 2=@n45.R, 3=@n45.R, 4=@n45.R, A=NamedRerouteUsage()) {materialFunction=/Engine/Functions/Engine_MaterialFunctions02/Lerp_Multiple_Float}.Lerp 4 Inputs) {name=Spectrum Exponent, nodeColor=<rgba 1 0 0.5625 1>}
  n46 = ScalarParameter() {group=Audio Reactivity - Gameplay Idle Cycle, parameterName=Idle Time Offset Back and Bottom Hearts}
  n47 = VectorParameter() {group=Audio Reactivity - Gameplay Idle Cycle, parameterName=Idle Time Offset Side Hearts, defaultValue=<rgba 0.1 0.2 0.3 0.4>}
  n48 = MaterialFunctionCall(1=@n47.R, 2=@n47.R, 3=@n47.R, 4=@n47.R, A=NamedRerouteUsage()) {materialFunction=&Lerp_Multiple_Float}
  n49 = NamedRerouteUsage()
  n50 = Time()
  n51 = ScalarParameter() {defaultValue=0.25, group=Audio Reactivity - Gameplay Idle Cycle, parameterName=Idle Timeline Speed}
  NamedRerouteDeclaration(None=Add(A=LinearInterpolate(A=@n46, B=@n48.Lerp 4 Inputs, Alpha=@n49), B=Multiply(A=@n50, B=@n51)) {desc="When idle on gameplay, cycles a basic animation through the hearts"}) {name=Idle Timeline, nodeColor=<rgba 0.5234 1 0 1>}
  n52 = ScalarParameter() {defaultValue=6, group=Audio Reactivity - Spectrum and Amplitude controls, parameterName=Amplitude Exponent Back and Bottom Hearts}
  NamedRerouteDeclaration(None=Power(Base=NamedRerouteUsage(), Exp=@n52)) {name=Amplitude Back and Bottom Hearts, nodeColor=<rgba 1 0 0.8438 1>}
 ]}
